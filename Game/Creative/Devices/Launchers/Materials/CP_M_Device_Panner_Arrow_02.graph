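# Translucent, unlit, two-sided spline-mesh material for Launcher devices that draws glowing arrows scrolling along a path.
# An arrow mask pans with tiling scaled by BP_SplineLength, fades at both ends, and multiplies procedural UV gradient bands.
# That mask blends tint colors with an emissive color, and a static switch can substitute a raw texture mask.
# SplineThicken optionally builds a camera-facing ribbon sized by the SplineWidth dynamic parameter, plus a fixed Z offset.
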Material /Game/Creative/Devices/Launchers/Materials/CP_M_Device_Panner_Arrow_02
graph {
 properties {
  bUsedWithSplineMeshes=true
  blendMode=BLEND_Translucent
  pixelDepthOffsetMode=PDOM_Legacy
  shadingModel=MSM_Unlit
  twoSided=true
  usedShadingModels=Unlit
 }
 nodes [
  n1 = DynamicParameter() {defaultValue=<rgba 100 1 1 1>, paramNames [SplineWidth,Param2,Param3,Param4]}
  Add(
   A=StaticSwitch(
    True=MaterialFunctionCall(WidthBase=@n1.SplineWidth, WidthTip=@n1.SplineWidth) {materialFunction=&SplineThicken}.WorldPositionOffset
    Value=StaticBoolParameter() {defaultValue=true, parameterName=CameraFacingSpline}
   )
   B=Constant3Vector() {constant=<rgba 0 0 25 0>}
  )
  n2 = TextureCoordinate()
  n3 = StaticSwitch(
   True=@n2
   False=MaterialFunctionCall(UVs=@n2, Rotation Center=Constant() {r=0.5}, "Rotation Angle (0-1)"=Constant() {r=0.5}) {materialFunction=&CustomRotator}.Rotated Values
   Value=StaticBoolParameter() {defaultValue=true, parameterName=Direction Switch}
  )
  n4 = Multiply(A=ComponentMask(None=@n3) {g=true}, B=ScalarParameter() {defaultValue=2.6, parameterName=Y tiling})
  Add(A=@n4, B=ComponentMask(None=ScalarParameter() {defaultValue=-0.01, parameterName=Y Offset}) {g=true})
  n5 = VectorParameter() {parameterName=Color2, defaultValue=<rgba 0.164 0 1 1>}
  n6 = ComponentMask(None=Saturate(None=@n3)) {r=true}
  n7 = OneMinus(None=@n6)
  n8 = Clamp(None=MaterialFunctionCall(In=Multiply(A=@n7, B=@n6), Contrast=ScalarParameter() {defaultValue=0.065, parameterName=Gradient_Mask_Contrast}) {materialFunction=&CheapContrast}.Result)
  n9 = ScalarParameter() {defaultValue=0.2, parameterName=Gradient Vertical Max}
  n10 = Subtract(A=@n9, B=Constant() {r=0.063})
  n11 = Multiply(A=Multiply(
    A=MaterialFunctionCall(Alpha=@n8, Min=@n10, Max=@n9) {materialFunction=&SmoothStep}.Result
    B=ScalarParameter() {defaultValue=1, parameterName=Gradient Vertical Opacity}
   ))
  n12 = ComponentMask(None=Saturate(None=@n3)) {r=true}
  n13 = Multiply(A=OneMinus(None=@n12), B=@n12)
  n14 = ScalarParameter() {defaultValue=0.25}
  n15 = Subtract(A=@n14, B=Constant() {r=0.02})
  n16 = Multiply(A=Multiply(
    A=MaterialFunctionCall(Alpha=@n13, Min=@n15, Max=@n14) {materialFunction=&SmoothStep}.Result
    B=ScalarParameter() {defaultValue=1, parameterName=Gradient Vertical Opacity}
   ))
  n17 = VectorParameter() {parameterName=Color, defaultValue=<rgba 1 0.4244 0 1>}
  n18 = ComponentMask(None=Saturate(None=@n3)) {g=true}
  n19 = LinearInterpolate(
   A=@n18
   B=MaterialFunctionCall(
    Alpha=@n18
    Min=ScalarParameter() {parameterName=Gradient Horizontal Min Top, sortPriority=34}
    Max=ScalarParameter() {defaultValue=1, parameterName=Gradient Horizontal Max Bottom, sortPriority=34}
   ) {materialFunction=&SmoothStep}.Result
  )
  n20 = Multiply(A=Multiply(A=@n19, B=OneMinus(None=@n19)), B=ScalarParameter() {defaultValue=2, parameterName=Gradient Horizontal Intensity})
  n21 = Multiply(
   A=Multiply(
    A=@n11
    B=MaterialFunctionCall(X=@n20, Y=@n20, Z=@n20) {materialFunction=&MakeFloat3}.Result
   )
   B=ScalarParameter() {defaultValue=1, parameterName=BG Opacity}
  )
  n22 = TextureCoordinate()
  n23 = ScalarParameter() {defaultValue=1, parameterName=BP_SplineLength}
  n24 = AppendVector(A=Constant() {r=1}, B=Divide(A=@n23, B=ScalarParameter() {defaultValue=500, parameterName=Arrow_Scale}))
  n25 = TextureSampleParameter2D(UVs=Panner(Coordinate=Multiply(A=@n22, B=@n24)) {speedY=1}) {parameterName=Texture_Alpha, samplerType=SAMPLERTYPE_Masks, texture=/Game/Creative/Devices/Launchers/Textures/CP_T_Device_Arrow_01}
  n26 = TextureCoordinate()
  n27 = Abs(None=ConstantBiasScale(None=ComponentMask(None=@n26) {g=true}) {bias=-0.5, scale=2})
  n28 = Divide(A=@n23) {constB=300}
  n29 = Multiply(A=@n25.RGB, B=OneMinus(None=Power(Base=@n27, Exp=@n28) {constExponent=8}))
  n30 = ScalarParameter() {defaultValue=1, parameterName=Arrow Opacity}
  n31 = LinearInterpolate(A=@n21, B=Multiply(A=@n21, B=Multiply(A=@n29, B=@n30)))
  LinearInterpolate(
   A=LinearInterpolate(A=Multiply(A=@n11, B=@n5.RGB), B=VectorParameter() {parameterName=Color3, defaultValue=<rgba 0.164 0 1 1>}.RGB, Alpha=OneMinus(None=@n16))
   B=Multiply(A=Multiply(A=@n17.RGB, B=@n31), B=ScalarParameter() {defaultValue=6, parameterName=Emissive})
   Alpha=@n31
  )
  n32 = Multiply(A=ComponentMask(None=@n3) {r=true}, B=ScalarParameter() {defaultValue=1, parameterName=X tiling})
  n33 = ScalarParameter() {defaultValue=1, parameterName=X Offset}
  Panner(
   Coordinate=AppendVector(A=Add(A=@n32, B=@n33), B=@n4)
   Speed=AppendVector(A=Constant(), B=ScalarParameter() {defaultValue=1.75, parameterName=Arrow Speed})
  )
  n34 = TextureSampleParameter2D() {parameterName=Texture_Alpha, samplerType=SAMPLERTYPE_Masks, texture=&CP_T_Device_Arrow_01}
  StaticSwitch(True=@n34.RGB, False=@n31, Value=StaticBoolParameter() {parameterName=OpacityByTexture})
 ]
}
